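# Sound cue that plays one Season 24 Battle Pass dialogue line spoken by the Evie voice, apparently for quest comms.
# A single dialogue player node is the root, resolving the dialogue wave with Evie as speaker and no targets.
# The cue routes to the narrator VO sound class, uses the shared quest comms attenuation with a very large max distance, and boosts volume.
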SoundCue /BattlePassS24/Audio/VO/Cues/EVIE_S24_REGS_12_01_010_Dialogue_Cue
root=DialoguePlayer_0
nodes [{
  id=DialoguePlayer_0
  properties {dialogueWaveParameter {
    dialogueWave=/BattlePassS24/Audio/VO/Waves/Evie/EVIE_S24_REGS_12_01_010_Dialogue
    context {speaker=/BattlePassS24/Audio/VO/Evie_DV, targets []}
   }}
 }]
properties {
 attenuationSettings=/BattlePassS24/Audio/VO/S24_Quest_Comms_Att
 maxDistance=2097152
 soundClassObject=/Game/Sounds/FortSoundClassesAndModes/Fort_VO_Narrator
 volumeMultiplier=2
}
