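# Sound cue for one Season 24 Battle Pass voice line spoken by the Thunder character.
# A single dialogue player node plays the matching dialogue wave, with the Thunder dialogue voice as speaker and no targets.
# The cue uses the quest comms attenuation and narrator VO sound class, with boosted volume and a very large max distance.
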SoundCue /BattlePassS24/Audio/VO/Cues/THUNDER_S24_REGS_04_02_070_Dialogue_Cue
root=DialoguePlayer_0
nodes [{
  id=DialoguePlayer_0
  properties {dialogueWaveParameter {
    dialogueWave=/BattlePassS24/Audio/VO/Waves/Thunder/THUNDER_S24_REGS_04_02_070_Dialogue
    context {speaker=/BattlePassS24/Audio/VO/Thunder_DV, targets []}
   }}
 }]
properties {
 attenuationSettings=/BattlePassS24/Audio/VO/S24_Quest_Comms_Att
 maxDistance=2097152
 soundClassObject=/Game/Sounds/FortSoundClassesAndModes/Fort_VO_Narrator
 volumeMultiplier=2
}
