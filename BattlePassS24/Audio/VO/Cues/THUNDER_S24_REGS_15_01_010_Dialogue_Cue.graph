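# Sound cue for a Thunder voice-over line in the BattlePassS24 audio content.
# A single dialogue player root plays the line's dialogue wave with the Thunder dialogue voice as speaker and no targets.
# An attenuation asset named for a Season 24 quest arcade cabinet spatializes or attenuates it, and the narrator VO sound class handles its mix grouping.
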SoundCue /BattlePassS24/Audio/VO/Cues/THUNDER_S24_REGS_15_01_010_Dialogue_Cue
root=DialoguePlayer_0
nodes [{
  id=DialoguePlayer_0
  properties {dialogueWaveParameter {
    dialogueWave=/BattlePassS24/Audio/VO/Waves/Thunder/THUNDER_S24_REGS_15_01_010_Dialogue
    context {speaker=/BattlePassS24/Audio/VO/Thunder_DV, targets []}
   }}
 }]
properties {
 attenuationSettings=/BattlePassS24/Audio/VO/S24_Quest_ArcadeCab_Att
 maxDistance=4000
 soundClassObject=/Game/Sounds/FortSoundClassesAndModes/Fort_VO_Narrator
 volumeMultiplier=1
}
